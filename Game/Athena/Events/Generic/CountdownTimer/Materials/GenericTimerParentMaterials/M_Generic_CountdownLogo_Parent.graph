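# Translucent unlit parent material for the generic event countdown timer's holographic logo.
# Tinted scrolling scanlines, with tiling adjusted by camera distance, are added to the SlateUI texture, with soft sphere masks and texture alpha shaping brightness and back glow.
# A static switch enables camera-facing sprite offset, pivoting on impostor UV-encoded position plus an HLOD offset and scaled by object Z scale and a UV3 channel.
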Material /Game/Athena/Events/Generic/CountdownTimer/Materials/GenericTimerParentMaterials/M_Generic_CountdownLogo_Parent
graph {
 properties {bAutomaticallySetUsageInEditor=false, blendMode=BLEND_Translucent, shadingModel=MSM_Unlit, usedShadingModels=Unlit}
 nodes [
  ActorPositionWS()
  n1 = MaterialFunctionCall() {materialFunction=/Game/Athena/Events/Generic/CountdownTimer/Materials/MaterialFunctions/MF_Generic_CountdownHologram}
  Add(A=@n1.Final WPO, B=BreakMaterialAttributes!(Attr=@n1.Material).WorldPositionOffset)
  n2 = Add(
   A=VectorParameter() {parameterName=PositionOffsetHLOD, defaultValue=<rgba 0 0 0 1>}.RGB
   B=MaterialFunctionCall() {materialFunction=/Game/ImpostorBaker/MaterialFunctions/ImpostorInfo_FromUVs}.Position
  )
  Divide(A=Distance(A=@n2, B=CameraPositionWS()), B=ScalarParameter() {defaultValue=2117.0803, parameterName=Bias})
  n3 = VectorParameter() {parameterName=ScanLineColor, defaultValue=<rgba 0.3438 0.1031 0.1647 1>}
  n4 = MaterialFunctionCall!(Distance=ScalarParameter() {defaultValue=2000, parameterName=CameraDisatanceControl}) {materialFunction=/Game/Effects/Fort_Effects/Materials/_Functions/MF_CameraDistanceFade}
  n5 = MaterialFunctionCall!(
   ScanLine_Tiling_1
   Scanline_PannerSpeed_1=Constant() {r=0.025}
   ScanLine_Tiling=Multiply(A=ScalarParameter() {defaultValue=100, parameterName=ScanLine_Tile}, B=LinearInterpolate(Alpha=@n4.Result) {constA=1, constB=0.35})
   Scanline_PannerSpeed
   Scanline_Rotation
   UVs
  ) {materialFunction=/Game/Global/Materials/MaterialGym/Materials/HolographicTexture/MF_HolographicTexture_ScanLines}
  n6 = TextureSampleParameter2D(UVs=TextureCoordinate()) {parameterName=SlateUI, samplerSource=SSM_Clamp_WorldGroupSettings, texture=/Engine/EngineResources/WhiteSquareTexture}
  n7 = SphereMask(A=TextureCoordinate() {vTiling=7}, B=Constant2Vector() {g=2.3, r=0.5}) {attenuationRadius=0.5, hardnessPercent=0}
  n8 = Add(A=Power(Base=OneMinus(None=@n7)) {constExponent=1}) {constB=0.25}
  Multiply(A=Add(A=Multiply(A=@n3.RGB, B=@n5.Result), B=@n6.RGB), B=Saturate(None=Add(A=@n6.A, B=@n8)))
  n9 = AppendVector(A=ScalarParameter() {defaultValue=200, parameterName=ScaleX}, B=ScalarParameter() {defaultValue=200, parameterName=ScaleY})
  Multiply(A=@n9, B=ScalarParameter() {defaultValue=0.5, parameterName=ScaleMin})
  Multiply(A=@n9, B=ScalarParameter() {defaultValue=1, parameterName=ScaleMax})
  n10 = SphereMask(A=TextureCoordinate() {vTiling=2}, B=Constant2Vector() {g=1, r=0.5}) {attenuationRadius=0.5, hardnessPercent=0}
  n11 = ScalarParameter() {defaultValue=0.25, parameterName=BackGlowOpacity}
  Saturate(None=Add(A=Multiply(A=@n10, B=@n11) {constB=0.65}, B=@n6.A))
  n12 = Multiply(
   A=MaterialFunctionCall() {materialFunction=/Engine/Functions/Engine_MaterialFunctions02/WorldPositionOffset/ObjectScale}.Scale Z
   B=ComponentMask(None=TextureCoordinate() {coordinateIndex=3}) {g=true}
  )
  StaticSwitchParameter(
   True=MaterialFunctionCall!(Center Location=@n2, 0-1 UVs, XY scale=Multiply(A=@n9, B=@n12), OverridePivot, "Normals (Optional)") {materialFunction=&Sprite}.World position offset
   False=Constant3Vector()
  ) {defaultValue=true, parameterName=IsCameraFacing}
 ]
}
