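# M_Generic_Countdown_Parent is a parent material for the generic event countdown timer, with a hologram look indicated by its function name.
# It is an unlit, translucent material in material-attributes mode, with responsive AA and instanced static mesh support enabled.
# Its graph holds a single call to MF_Generic_CountdownHologram, so the shading logic lives in that function rather than here.
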Material /Game/Athena/Events/Generic/CountdownTimer/Materials/GenericTimerParentMaterials/M_Generic_Countdown_Parent
graph {
 properties {
  bEnableResponsiveAA=true
  bUseMaterialAttributes=true
  bUsedWithInstancedStaticMeshes=true
  blendMode=BLEND_Translucent
  pixelDepthOffsetMode=PDOM_Legacy
  shadingModel=MSM_Unlit
  usedShadingModels=Unlit
 }
 nodes [MaterialFunctionCall() {materialFunction=/Game/Athena/Events/Generic/CountdownTimer/Materials/MaterialFunctions/MF_Generic_CountdownHologram}]
}
